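# Unlit, masked, two-sided master material for the inside of a rift portal, showing a distorted interior image with glowing, rippling edges.
# Interior UVs come from screen space or a virtual plane, then get pixelation, panning noise and ripples, with chromatic offsets or optional parallax texture blending.
# Switches add variant overlays like fog, stars and sun streaks, with time-of-day tint.
# Portal open functions drive world position offset and dithered opacity.
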Material /PlotTwist/Items/RiftGate/Assets/ProtoAssets/M_Pickle_Port_Inner_Master_Yogurt
graph {
 properties {
  bCastRayTracedShadows=false
  blendMode=BLEND_Masked
  pixelDepthOffsetMode=PDOM_Legacy
  shadingModel=MSM_Unlit
  twoSided=true
  usedShadingModels=Unlit
 }
 nodes [
  n1 = ScalarParameter() {parameterName=Virtual Plane Bump Offset}
  n2 = AppendVector(A=ScalarParameter() {group=WIP, parameterName=UVOffset_X}, B=ScalarParameter() {group=WIP, parameterName=UVOffset_Y})
  n3 = Constant2Vector() {g=1, r=0.5}
  n4 = NamedRerouteUsage()
  n5 = MaterialFunctionCall!(
   Heightmap Texture=TextureObjectParameter() {parameterName=Depth Mask, samplerType=SAMPLERTYPE_Masks, texture=&T_BlackMask}
   Height Ratio=@n1
   Min Steps=ScalarParameter() {defaultValue=2, parameterName=Min Steps Prallax}
   Max Steps=ScalarParameter() {defaultValue=8, parameterName=Max Steps Prallax}
   UVs=Add(A=Multiply(A=@n2, B=@n3), B=Multiply(A=@n4, B=@n3))
   Heightmap Channel=Constant4Vector() {constant=<rgba 1 0 0 0>}.RGBA
   Reference Plane=ScalarParameter() {parameterName=Virtual Plane Reference}
   -----------------------------
   Use World Coordinates
   Specify Manual Texture Size
   Manual Texture Size
   ------------------------------
   "Render Shadows (Occlusion Mapping)"
   Light Vector
   Shadow Steps
   Shadow Penumbra
   Transform To VertexNormal
  ) {materialFunction=/Engine/Functions/Engine_MaterialFunctions01/Texturing/ParallaxOcclusionMapping}
  Clamp(None=@n5.Offset Only)
  Multiply(A=@n5.Offset Only)
  n6 = MaterialFunctionCall!(SphereRadius, SphereCenter) {materialFunction=/Game/Athena/Environments/Nexus/Pickles/Materials/MF_Pickle_Port_OpenWPO}
  NamedRerouteDeclaration(None=@n6.DistanceFade) {name=Distance Fade, nodeColor=<rgba 0 1 0.0156 1>}
  NamedRerouteDeclaration(None=@n6.GlowAdd) {name=Add Glow, nodeColor=<rgba 1 0.0234 0 1>}
  NamedRerouteDeclaration(None=Add(A=@n6.PortalWPO, B=@n6.PortalWPO)) {name=Port OpenWPO, nodeColor=<rgba 0.6172 1 0 1>}
  n7 = Add(A=@n2, B=@n4)
  n8 = Add(A=@n7, B=@n5.Offset Only)
  n9 = QualitySwitch(
   Default=BumpOffset(Coordinate=@n7, Height=@n1, HeightRatioInput=ScalarParameter() {defaultValue=0.05, group=WIP, parameterName=HeightRatioInput}) {heightRatio=0.01}
   High=@n8
   Epic=@n8
  )
  n10 = Multiply(
   A=Multiply(A=@n9, B=ScalarParameter() {defaultValue=8, group=WIP, parameterName=Distortion_Tiling}) {constB=8}
   B=Constant2Vector() {g=0.5, r=1}
  )
  n11 = Multiply(A=Time(), B=ScalarParameter() {defaultValue=1, group=WIP, parameterName=DistSpeed})
  n12 = Add(A=@n10, B=@n11)
  n13 = Add(A=Multiply(A=@n10) {constB=0.8}, B=Multiply(A=@n11, B=Constant2Vector() {g=1, r=-1}))
  n14 = Multiply(A=Add(
    A=TextureSample(UVs=@n12) {texture=/Game/Effects/Fort_Effects/Textures/Tile/T_Smooth_Tile_Chromatic}.R
    B=TextureSample(UVs=@n13) {texture=&T_Smooth_Tile_Chromatic}.G
   )) {constB=0.5}
  n15 = ScalarParameter() {defaultValue=2, group=WIP, parameterName=distortpow}
  n16 = ScalarParameter() {defaultValue=0.25, group=WIP, parameterName=centX}
  n17 = ScalarParameter() {defaultValue=0.25, group=WIP, parameterName=centY}
  n18 = Constant2Vector() {g=1, r=0.5}
  n19 = OneMinus(None=SphereMask(
    A=AppendVector(A=@n16, B=@n17)
    B=MaterialFunctionCall(UVs=@n9, Texture Scale=@n18) {materialFunction=&ScaleUVsByCenter}.UVs
    Radius=ScalarParameter() {defaultValue=0.3, group=WIP, parameterName=rad}
    Hardness=ScalarParameter() {group=WIP, parameterName=hard}
   ) {attenuationRadius=0.2, hardnessPercent=0})
  n20 = ScalarParameter() {defaultValue=0.1, group=WIP, parameterName=distortbase}
  n21 = Multiply(A=Multiply(A=Power(Base=@n14, Exp=@n15), B=Max(A=@n19, B=@n20)), B=ScalarParameter() {group=WIP, parameterName=distortstrenght2})
  n22 = Multiply(A=@n19, B=ScalarParameter() {defaultValue=8, group=WIP, parameterName=MipFactor}) {constB=4}
  n23 = TextureSampleParameter2D!(UVs=Add(A=@n21, B=@n9), Level=@n22, Apply View MipBias) {mipValueMode=TMVM_MipLevel, parameterName=InteriorTextureA, texture=&T_BlacksRGB}
  n24 = TextureSampleParameter2D(UVs=Clamp(None=Add(A=@n21, B=@n9)) {maxDefault=0.99, minDefault=0.01}, Level=@n22) {mipValueMode=TMVM_MipLevel, parameterName=InteriorTextureB, texture=&T_BlacksRGB}
  n25 = ScalarParameter() {parameterName=InteriorBlend}
  n26 = VectorParameter() {parameterName=PlaneNormal, defaultValue=<rgba 0.1 0 -1 1>}
  n27 = Add(A=ActorPositionWS(), B=VectorParameter() {parameterName=WaterPlaneCenter, defaultValue=<rgba 0 0 1000 1>}.RGB)
  n28 = VectorParameter() {parameterName=WaterTile, defaultValue=<rgba 15000 15000 0 0>}
  n29 = MaterialFunctionCall(
   Plane Normal Axis=Normalize(VectorInput=Transform(None=@n26.RGB) {transformSourceType=TRANSFORMSOURCE_Local})
   Plane X Axis=Constant3Vector() {constant=<rgba 1 0 0 0>}
   Plane Center=@n27
   UV Scale=AppendVector(A=@n28.R, B=@n28.R)
  ) {materialFunction=&VirtualPlaneCoordinates}
  n30 = TextureSampleParameter2D!(UVs=Panner(Coordinate=@n29.UVs) {speedX=0.05, speedY=0.02}, Apply View MipBias) {parameterName=WaterNormal, samplerType=SAMPLERTYPE_Normal, texture=/Game/Athena/Apollo/Water/Textures/02_Rounded/02_WaterFoam_Veroni_base_N}
  n31 = TextureSampleParameter2D(UVs=Panner(Coordinate=Multiply(A=@n29.UVs) {constB=1.2}) {speedX=0.01, speedY=-0.02}) {parameterName=WaterNormal, samplerType=SAMPLERTYPE_Normal, texture=&02_WaterFoam_Veroni_base_N}
  n32 = ScalarParameter() {defaultValue=50000, parameterName=WaterNormalRadius}
  n33 = Multiply(
   A=Multiply(A=VectorParameter() {parameterName=WaterNormalMult, defaultValue=<rgba 0.5 0.5 1 1>}.RGB, B=LinearInterpolate(A=@n30.RGB, B=@n31.RGB))
   B=Power(Base=SphereMask(A=@n29.World Position, B=@n27, Radius=@n32) {hardnessPercent=0})
  )
  n34 = VertexColor()
  n35 = ScalarParameter() {defaultValue=1, parameterName=EdgeMaskPower}
  n36 = ScalarParameter() {parameterName=ExgeMaskPowerMin}
  n37 = MaterialFunctionCall() {materialFunction=&CameraDirectionVector}
  n38 = Transform(None=Constant3Vector() {constant=<rgba 0 1 0 0>}) {transformSourceType=TRANSFORMSOURCE_Instance}
  n39 = ScalarParameter() {defaultValue=0.9, parameterName=Virtual Plane Camera/Local Vector Blend}
  n40 = MaterialFunctionCall(
   Plane Normal Axis=Normalize(VectorInput=LinearInterpolate(A=@n37.Result, B=@n38, Alpha=@n39))
   Plane Center=TransformPosition(None=VectorParameter() {parameterName=Image Offset, defaultValue=<rgba 0 0 0 100>}.RGB) {transformSourceType=TRANSFORMPOSSOURCE_Instance, transformType=TRANSFORMPOSSOURCE_World}
   UV Scale=ScalarParameter() {defaultValue=200, parameterName=Image Size}
  ) {materialFunction=&VirtualPlaneCoordinates}
  n41 = ScalarParameter() {defaultValue=5000, parameterName=DepthOffset}
  n42 = LinearInterpolate(Alpha=Saturate(None=TwoSidedSign())) {constA=1, constB=-1}
  n43 = ScalarParameter() {defaultValue=1000, parameterName=HeightOffsetBase}
  n44 = ScalarParameter() {defaultValue=-1000, parameterName=HeightOffsetTop}
  n45 = ComponentMask(None=CameraPositionWS()) {b=true}
  n46 = ComponentMask(None=ActorPositionWS()) {b=true}
  n47 = Saturate(None=Divide(A=Subtract(A=@n45, B=@n46), B=ScalarParameter() {defaultValue=3000, parameterName=HeightOffsetMaskSize}))
  n48 = TransformPosition(None=MaterialFunctionCall(Y=Multiply(A=@n41, B=@n42), Z=LinearInterpolate(A=@n43, B=@n44, Alpha=@n47) {constB=0}) {materialFunction=&MakeFloat3}.Result) {transformType=TRANSFORMPOSSOURCE_World}
  n49 = Add(A=Subtract(
    A=ScreenPosition().ViewportUV
    B=MaterialFunctionCall(World Position=@n48) {materialFunction=&TransformToClipSpace}.Buffer UVs for Scene Textures
   )) {constB=0.5}
  n50 = ScalarParameter() {defaultValue=1200, parameterName=DistanceScale}
  n51 = Distance(A=ActorPositionWS(), B=CameraPositionWS())
  n52 = Constant() {r=1}
  n53 = MaterialFunctionCall(Float2=ViewSize()) {materialFunction=&BreakOutFloat2Components}
  n54 = Divide(A=@n53.R, B=@n53.R)
  n55 = ScalarParameter() {defaultValue=0.8, parameterName=ScaleMult}
  n56 = Multiply(A=NamedRerouteDeclaration(None=MaterialFunctionCall!(FlickerInitialSpeed, FlickerTimeOffset) {materialFunction=/Game/Athena/Environments/Nexus/Pickles/Materials/MF_Pickle_Port_Shudder}.Result) {name=ShudderResult, nodeColor=<rgba 0.4129 0.0452 1 1>}) {constB=-0.025}
  n57 = Multiply(
   A=Clamp(None=Divide(A=@n50, B=@n51), Max=ScalarParameter() {defaultValue=0.75, parameterName=MaxDistanceScaleUp}) {clampMode=CMODE_ClampMax, maxDefault=0}
   B=Multiply(A=AppendVector(A=@n52, B=@n54), B=Add(A=@n55, B=@n56))
  )
  n58 = MaterialFunctionCall!(
   UVs=StaticSwitchParameter(
    True=NamedRerouteDeclaration(None=@n40.UVs) {name=Virtual Plane, nodeColor=<rgba 1 0.7031 0 1>}
    False=MaterialFunctionCall(UVs=@n49, Texture Scale=@n57) {materialFunction=&ScaleUVsByCenter}.UVs
   ) {parameterName=Use Virtual Plane Coordionates}
   Resolution=Multiply(A=ScalarParameter() {defaultValue=100, parameterName=PixelRes}) {constB=0.5}
  ) {materialFunction=/Game/Athena/Environments/Nexus/Pickles/Materials/MF_Skirmish_Pixelate}
  n59 = Multiply(A=@n58.Result) {constB=3}
  n60 = TextureCoordinate() {uTiling=4, vTiling=2}
  n61 = Time()
  n62 = ScalarParameter() {defaultValue=2, parameterName=PanSpeedDistortion}
  n63 = TextureSample(UVs=Panner(Coordinate=LinearInterpolate(A=@n59, B=@n60) {constAlpha=0.6}, Time=Multiply(A=@n61, B=@n62)) {speedX=0.01, speedY=-0.07}) {texture=&T_Smooth_Tile_Chromatic}
  n64 = Constant2Vector() {g=1, r=0.05}
  n65 = VectorParameter() {parameterName=DistortionTile, defaultValue=<rgba 0.3 0.1 2 0>}
  n66 = AppendVector(A=@n65.R, B=@n65.R)
  n67 = Add(A=Multiply(A=@n60, B=Multiply(A=@n64, B=@n66)), B=Constant2Vector() {r=1.75})
  n68 = Multiply(A=@n60, B=@n66)
  n69 = NamedRerouteUsage()
  n70 = Time()
  n71 = ScalarParameter() {defaultValue=0.1, parameterName=PanSpeed}
  n72 = TextureSampleParameter2D(UVs=Panner(Coordinate=LinearInterpolate(A=@n67, B=@n68, Alpha=@n69) {constA=4}, Time=Multiply(A=@n70, B=@n71)) {speedY=-0.2}) {parameterName=DistortionTexture, texture=/Game/Effects/Fort_Effects/Textures/Tile/T_Caustics_Chromatic}
  n73 = Desaturation(None=LinearInterpolate(A=@n63.RGB, B=@n72.RGB))
  n74 = NamedRerouteUsage()
  n75 = Add(
   A=Power(Base=OneMinus(None=@n34.G), Exp=LinearInterpolate(A=@n35, B=@n36, Alpha=@n73) {constA=3, constB=0.25}) {constExponent=0.5}
   B=Multiply(A=@n74) {constB=0.2}
  )
  n76 = SphereMask(
   A=ComponentMask(None=TextureCoordinate()) {r=true}
   B=Constant() {r=0.5}
   Radius=ScalarParameter() {defaultValue=0.5, parameterName=InteriorLinesRadius}
  ) {attenuationRadius=500, hardnessPercent=0}
  n77 = LinearInterpolate(Alpha=@n76) {constA=1.6, constB=-0.4}
  n78 = Power(Base=@n76, Exp=ScalarParameter() {defaultValue=3, parameterName=InteriorLinesPower})
  n79 = ComponentMask(None=TextureCoordinate()) {g=true}
  n80 = VectorParameter() {parameterName=LinesTile, defaultValue=<rgba 5 3 0 0>}
  n81 = Multiply(A=@n79, B=@n80.R)
  n82 = Panner(Coordinate=Multiply(A=AppendVector(A=@n78, B=@n79), B=AppendVector(A=@n81, B=@n80.G)) {constB=2}) {speedX=-0.1, speedY=-0.17}
  n83 = QualitySwitch(
   Default=Multiply(A=Saturate(None=@n77), B=TextureSample(UVs=@n82) {texture=/Game/Athena/Environments/Nexus/Textures/T_ZeroPoint_MagneticWaves}.RGB) {constB=0}
   Low=Constant()
  )
  n84 = ComponentMask(None=@n83) {g=true, r=true}
  n85 = ScalarParameter() {defaultValue=0.3, parameterName=EdgeRippleDistortion}
  n86 = NamedRerouteDeclaration(None=Multiply(A=@n75, B=Multiply(A=@n84, B=@n85))) {name=Ripples, nodeColor=<rgba 0 1 0.8359 1>}
  n87 = Add(
   A=@n58.Result
   B=Multiply(A=ComponentMask(None=@n63.RGB) {g=true, r=true}, B=ScalarParameter() {defaultValue=0.017, parameterName=MainDistortion})
  ) {constB=0}
  n88 = NamedRerouteDeclaration(None=StaticSwitchParameter(True=Add(A=@n87, B=@n86) {constB=0}, False=@n87) {defaultValue=true, parameterName=UseRipples}) {name=Distorted UVs, nodeColor=<rgba 0.0078 0 1 1>}
  n89 = Add(A=ShadingPathSwitch(Default=ComponentMask(None=@n33) {g=true, r=true}, Mobile=Constant()), B=@n88)
  n90 = TextureCoordinate()
  n91 = Normalize(VectorInput=AppendVector(A=Subtract(A=@n90) {constB=0.5}, B=Constant()))
  n92 = Desaturation(None=Multiply(A=@n83), Fraction=Constant() {r=-1})
  n93 = Add(A=@n63.RGB, B=StaticSwitchParameter(True=Add(A=@n72.RGB, B=@n92), False=@n72.RGB) {defaultValue=true, parameterName=UseRipples})
  n94 = Add(A=@n93, B=@n83)
  n95 = Multiply(A=@n91, B=Multiply(A=ScalarParameter() {defaultValue=0.005, parameterName=ChromaticAmount}, B=Add(A=@n94) {constB=1.5})) {constB=2}
  n96 = MaterialFunctionCall(Float3=Multiply(A=@n75, B=StaticSwitchParameter(True=@n95, False=@n91) {defaultValue=true, parameterName=UseRipples})) {desc=Chromatic offset, materialFunction=&BreakOutFloat3Components}
  n97 = Add(A=@n88, B=@n96.R)
  n98 = Add(A=@n88, B=@n96.G)
  n99 = TextureSampleParameter2D(UVs=@n88) {parameterName=InteriorTexture, texture=/PlotTwist/Items/RiftGate/Assets/Textures/T_Rift2_A}
  n100 = StaticSwitchParameter(
   True=TextureSampleParameter2D(UVs=NamedRerouteDeclaration(None=@n89) {name=InteriorTextureUV, nodeColor=<rgba 1 0 0.75 1>}) {parameterName=InteriorTexture, texture=&T_Rift2_A}.RGB
   False=MaterialFunctionCall(
    R=TextureSampleParameter2D(UVs=@n97) {parameterName=InteriorTexture, texture=&T_Rift2_A}.R
    G=TextureSampleParameter2D(UVs=@n98) {parameterName=InteriorTexture, texture=&T_Rift2_A}.G
    B=@n99.B
   ) {materialFunction=&AppendMany}.RGB
  ) {parameterName=Dusty}
  n101 = ShadingPathSwitch(Default=@n100, Forward=@n99.RGB, Mobile=@n99.RGB)
  NamedRerouteDeclaration(None=StaticSwitchParameter(True=LinearInterpolate(A=@n23.RGB, B=@n24.RGB, Alpha=@n25), False=QualitySwitch(Default=@n101, Low=@n99.RGB, Medium=@n100)) {parameterName=UseInteriorTextureTransition}) {name=InteriorTexture, nodeColor=<rgba 0.0781 1 0 1>}
  n102 = OneMinus(None=ComponentMask(None=TextureCoordinate()) {g=true})
  n103 = ScalarParameter() {defaultValue=8, parameterName=Edge Opacity G Mask Exp}
  n104 = Clamp(None=Multiply(A=Power(Base=@n102, Exp=@n103) {constExponent=5}, B=ScalarParameter() {defaultValue=10, parameterName=Edge Opacity G Mask Mult}) {constB=10})
  n105 = ComponentMask(None=TextureCoordinate()) {r=true}
  n106 = Constant2Vector() {g=0.5, r=0.5}
  n107 = ScalarParameter() {defaultValue=0.35, parameterName=Opacity Mask Radius R}
  n108 = ScalarParameter() {defaultValue=1, parameterName=Opacity Mask Hardness R}
  n109 = ComponentMask(None=TextureCoordinate()) {g=true}
  n110 = ScalarParameter() {defaultValue=0.35, parameterName=Opacity Mask Radius G}
  n111 = ScalarParameter() {defaultValue=0.7, parameterName=Opacity Mask Hardness G}
  n112 = Multiply(
   A=Min(
    A=SphereMask(A=@n105, B=@n106.R, Radius=@n107, Hardness=@n108) {hardnessPercent=50}
    B=SphereMask(A=@n106.G, B=@n109, Radius=@n110, Hardness=@n111) {hardnessPercent=50}
   )
   B=NamedRerouteUsage()
  )
  NamedRerouteDeclaration(None=MaterialFunctionCall(Alpha Threshold=Multiply(A=@n104, B=@n112)) {materialFunction=&DitherTemporalAA}.Result) {name=Edge Mask Opacity, nodeColor=<rgba 0 1 0.3672 1>}
  n113 = NamedRerouteUsage()
  n114 = Add(A=@n113, B=NamedRerouteUsage())
  NamedRerouteDeclaration(None=MaterialFunctionCall!(
    UV=StaticSwitchParameter(True=@n114, False=@n113) {defaultValue=true, parameterName=UseRipples}
    Center
    K1=ScalarParameter() {parameterName=Mustache UV}
    K2
    P1
    P2
    P3
    P4
    UV Scale Bias=ScalarParameter() {parameterName=Mustache UV Scale}
    Enable K2
    Mirror Tiling=StaticBoolParameter() {group=WIP, parameterName=MirrorTiling}
    Clamp Tiling=StaticBoolParameter() {defaultValue=true, group=WIP, parameterName=ClampTiling}
   ) {materialFunction=/Game/Global/MaterialFunctions/MF_MustacheDistortion}.Result) {name=Mustache UVs, nodeColor=<rgba 0 0.2734 1 1>}
  n115 = NamedRerouteUsage()
  n116 = TextureCoordinate()
  n117 = ComponentMask(None=@n116) {r=true}
  n118 = Constant() {r=0.5}
  n119 = OneMinus(None=Multiply(A=Distance(A=@n117, B=@n118)) {constB=2})
  n120 = Distance(A=ComponentMask(None=@n116) {g=true}, B=Constant())
  n121 = Multiply(A=Multiply(A=Sine(None=@n119) {period=3}, B=Multiply(A=@n120) {constB=2}), B=ScalarParameter() {defaultValue=50, parameterName=Stuff})
  n122 = Constant3Vector() {constant=<rgba 0 1 0 0>}
  NamedRerouteDeclaration(None=Multiply(A=Multiply(A=@n115, B=@n121), B=Transform(None=@n122) {transformSourceType=TRANSFORMSOURCE_Instance})) {name=Bubble WPO, nodeColor=<rgba 0.7109 1 0 1>}
  NamedRerouteDeclaration(None=@n6.OpenAmount) {name=OpenAmount, nodeColor=<rgba 1 0 0.9656 1>}
  n123 = VectorParameter() {parameterName=Sunoffset, defaultValue=<rgba 0.21 0.35 0.5 0>}
  n124 = Add(A=Multiply(A=@n87, B=@n123.B), B=AppendVector(A=@n123.R, B=@n123.R))
  n125 = Rotator(Coordinate=@n124)
  n126 = Rotator(Coordinate=@n124) {speed=-0.2}
  n127 = Multiply(
   A=Multiply(
    A=TextureSampleParameter2D(UVs=@n125) {parameterName=SunTexture, texture=/Game/Effects/Fort_Effects/Textures/Flares/T_Flare_Streaks}.RGB
    B=TextureSampleParameter2D(UVs=@n126) {parameterName=SunTexture, texture=&T_Flare_Streaks}.RGB
   )
   B=VectorParameter() {parameterName=SunBrightness, defaultValue=<rgba 0.2 0.2 0.2 0>}.RGB
  ) {constB=-1}
  n128 = TextureSampleParameter2D(UVs=@n87) {parameterName=DustyMask, texture=&T_BlacksRGB}
  n129 = NamedRerouteUsage()
  n130 = ScalarParameter() {parameterName=WombatFogMaskOffset}
  n131 = Multiply(A=ScalarParameter() {defaultValue=1, parameterName=WombatFogMaskSize}, B=Saturate(None=ComponentMask(None=@n87) {g=true}))
  n132 = Saturate(None=Sine(None=Add(A=@n130, B=@n131)) {period=0.5})
  n133 = VectorParameter() {parameterName=WombatFogTile, defaultValue=<rgba 1.8 3.5 5 0>}
  n134 = AppendVector(A=@n133.R, B=@n133.R)
  n135 = VectorParameter() {parameterName=WombatFogPanDirection, defaultValue=<rgba 0.02 0.12 0.15 0>}
  n136 = Multiply(A=Time(), B=@n135.B)
  n137 = Panner(
   Coordinate=MaterialFunctionCall(XY=Multiply(A=@n134, B=@n87) {constB=2}) {materialFunction=&Swizzle}.YX
   Time=@n136
   Speed=AppendVector(A=@n135.R, B=@n135.R)
  ) {speedX=0.02, speedY=0.15}
  n138 = TextureSample(UVs=Panner(Coordinate=Multiply(A=@n87, B=@n133.B) {constB=5}, Time=@n136) {speedX=-0.08, speedY=-0.04}) {texture=&T_Smooth_tile}
  n139 = VectorParameter() {parameterName=WombatFogColor, defaultValue=<rgba 1 1 1 0>}
  n140 = Multiply(
   A=Multiply(
    A=TextureSampleParameter2D(UVs=@n137) {parameterName=WombatFog, texture=/Game/Packages/Fortress_Sky/SkyDome/Master/T_CloudWhisp_Pattern_04}.G
    B=LinearInterpolate(Alpha=@n138.G) {constA=0.3, constB=1.2}
   )
   B=@n139.RGB
  )
  n141 = Saturate(None=MaterialFunctionCall!(FlickerInitialSpeed=ScalarParameter() {defaultValue=0.7, parameterName=FlickerSpeed2}, FlickerRandom=Constant()) {materialFunction=/Game/Effects/Fort_Effects/Materials/_Functions/MF_Flicker}.Result)
  n142 = Panner(Coordinate=NamedRerouteUsage()) {speedX=0.2}
  n143 = Multiply(
   A=Add(
    A=TextureSampleParameter2D(UVs=@n142) {parameterName=FlickerTextureMask, texture=&T_BlacksRGB}.R
    B=Time()
   )
   B=ScalarParameter() {defaultValue=1, parameterName=FlickerSpeed}
  )
  n144 = Sine(None=@n143) {period=0.52}
  n145 = Add(A=Add(A=Sine(None=@n143) {period=0.81}, B=Sine(None=@n143) {period=3.5})) {constB=1.8}
  n146 = Saturate(None=Multiply(A=Add(A=@n144, B=@n145)) {constB=3})
  n147 = StaticSwitchParameter(True=Clamp(None=LinearInterpolate(A=@n141, Alpha=@n146), Min=@n139.A) {clampMode=CMODE_ClampMin}, False=Constant() {r=1}) {parameterName=Typhoon}
  n148 = QualitySwitch(Default=Multiply(A=Power(Base=@n132), B=Multiply(A=@n140, B=@n147)), Low=Constant())
  n149 = TextureSample(UVs=@n87) {samplerType=SAMPLERTYPE_Grayscale, texture=/Game/Athena/Environments/Nexus/Pickles/Textures/T_Pickle_Int_01_M}
  n150 = Time()
  n151 = ScalarParameter() {defaultValue=0.25, parameterName=CherryAnimSpeed}
  n152 = Sine(None=Add(A=@n138.G, B=Multiply(A=@n150, B=@n151)))
  n153 = Multiply(
   A=VectorParameter() {parameterName=BrightStarColor, defaultValue=<rgba 25 20.9249 10.8103 0>}.RGB
   B=Add(A=@n149.R, B=Multiply(A=@n149.R, B=@n152))
  ) {constB=15}
  n154 = Add(A=@n129) {constB=0.1}
  n155 = StaticSwitchParameter(True=Add(A=Multiply(A=@n153, B=@n154), B=Multiply(A=@n129) {constB=0.4}), False=@n129) {parameterName=Cherry}
  n156 = StaticSwitchParameter(
   True=Add(A=@n129, B=Multiply(A=@n127, B=@n128.G))
   False=StaticSwitchParameter(True=Add(A=@n148, B=@n155), False=@n155) {parameterName=Wombat}
  ) {parameterName=Dusty}
  n157 = MaterialFunctionCall(
   Color_Morning=VectorParameter() {parameterName=MorningTint, defaultValue=<rgba 0.4 0.3098 0.2735 0>}.RGB
   Color_Day=VectorParameter() {parameterName=DayTint, defaultValue=<rgba 1 1 1 0>}.RGB
   Color_Evening=VectorParameter() {parameterName=EveningTint, defaultValue=<rgba 0.5 0.4215 0.4165 0>}.RGB
   Color_Night=VectorParameter() {parameterName=NightTint, defaultValue=<rgba 1.5 1.3961 0.9576 0>}.RGB
  ) {materialFunction=&Color_Time_of_Day}
  n158 = Multiply(
   A=LinearInterpolate(
    A=VectorParameter() {parameterName=EdgeColor, defaultValue=<rgba 1.1639 0.8936 2.5 0>}.RGB
    B=VectorParameter() {parameterName=EdgeColorDestabilize, defaultValue=<rgba 0.868 0.3313 1 0>}.RGB
    Alpha=@n74
   )
   B=Add(A=Multiply(A=@n74, B=@n93), B=@n93)
  )
  n159 = Multiply(A=@n75, B=@n75)
  n160 = Multiply(A=Add(A=Multiply(A=@n156, B=@n157.Result), B=Multiply(A=@n158, B=@n159)), B=NamedRerouteUsage())
  n161 = NamedRerouteUsage()
  n162 = Saturate(None=TwoSidedSign())
  Reroute(None=StaticSwitchParameter(True=LinearInterpolate(A=@n160, B=@n161, Alpha=@n162), False=@n160) {parameterName=UseBubbleWPO})
  n163 = NamedRerouteUsage()
  n164 = Power(Base=@n6.OpenAmount) {constExponent=0.5}
  n165 = ScalarParameter() {parameterName=VertColorAlphaContrast}
  n166 = Multiply(
   A=Multiply(
    A=NamedRerouteDeclaration(None=@n164) {name=Open Amount Opacity, nodeColor=<rgba 1 0.4922 0 1>}
    B=MaterialFunctionCall(In=@n34.G, Contrast=@n165) {materialFunction=&CheapContrast}.Result
   )
   B=ScalarParameter() {defaultValue=1, parameterName=OpacityMult}
  )
  n167 = SphereMask(
   A=@n87
   B=Constant() {r=0.5}
   Radius=ScalarParameter() {defaultValue=0.65, parameterName=OpacitySphereRadius}
   Hardness=ScalarParameter() {defaultValue=0.3, parameterName=OpacitySphereHardness}
  ) {hardnessPercent=0.7}
  n168 = MaterialFunctionCall(Alpha Threshold=Multiply(A=@n6.WeldFade, B=Multiply(A=@n166, B=@n167))) {materialFunction=&DitherTemporalAA}
  n169 = StaticSwitchParameter(
   True=NamedRerouteUsage()
   False=StaticSwitchParameter(True=Multiply(A=@n163, B=@n168.Result), False=@n168.Result) {parameterName=UseDistanceFade}
  ) {parameterName=Use Mesh Edge Mask Opacity}
  n170 = Abs(None=Subtract(A=TwoSidedSign()))
  StaticSwitchParameter(True=Multiply(A=@n169, B=Saturate(None=@n170)), False=@n169) {group=WIP, parameterName=HidePositiveSign}
  n171 = Add(
   A=NamedRerouteUsage()
   B=MaterialFunctionCall(Offset Amount=ScalarParameter() {defaultValue=-100, parameterName=CamOffset}) {materialFunction=/Engine/Functions/Engine_MaterialFunctions02/WorldPositionOffset/CameraOffset}.Camera Offset
  )
  n172 = StaticSwitchParameter(True=Add(A=@n171, B=NamedRerouteUsage()), False=@n171) {parameterName=UseBubbleWPO}
  n173 = MaterialFunctionCall(Y=Multiply(A=VertexColor().R, B=ScalarParameter() {defaultValue=100, group=WIP, parameterName=Thickness}) {constB=50}) {materialFunction=&MakeFloat3}
  StaticSwitchParameter(True=Add(A=@n172, B=Transform(None=@n173.Result) {transformSourceType=TRANSFORMSOURCE_Local}), False=@n172) {group=WIP, parameterName=UseThickness}
 ]
}
